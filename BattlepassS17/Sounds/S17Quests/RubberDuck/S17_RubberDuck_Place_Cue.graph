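# Sound cue for placing a rubber duck, part of the Battle Pass Season 17 quest sounds.
# A random node picks among four equally weighted rubber duck placement wave variations.
# It plays under the quest sound class with the shared Season 17 quest attenuation asset for spatial falloff.
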SoundCue /BattlepassS17/Sounds/S17Quests/RubberDuck/S17_RubberDuck_Place_Cue
root=Random_0
nodes [
 {id=Random_0, children [WavePlayer_0,WavePlayer_1,WavePlayer_2,WavePlayer_3], properties {weights [1,1,1,1]}}
 {id=WavePlayer_0, properties {soundWaveAssetPtr=&/S17_RubberDuckPlace_01}}
 {id=WavePlayer_1, properties {soundWaveAssetPtr=&/S17_RubberDuckPlace_02}}
 {id=WavePlayer_2, properties {soundWaveAssetPtr=&/S17_RubberDuckPlace_03}}
 {id=WavePlayer_3, properties {soundWaveAssetPtr=&/S17_RubberDuckPlace_04}}
]
properties {
 attenuationSettings=/BattlepassS17/Sounds/S17_Quest_Att
 maxDistance=4000
 soundClassObject=&Fort_Quests
}
